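; PA announcement actor that repeatedly plays an assigned sound at its own location at random intervals.
; Initialize stores the sound and time range, then starts a random-delay timer only on a dedicated server.
; Each timer firing calls Client Play Audio and schedules the next with a new random delay.
; Client Play Audio plays the sound only where cosmetic events can execute and the actor is not marked for deletion.
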(event Custom|Server%20Play%20Audio
  (bind (_randtimerange_x _randtimerange_y) (Math|Vector2D|BreakVector2D (Variables|GetRandTimeRange_X)))
  (CallFunction|ClientPlayAudio)
  (Utilities|Time|SetTimerbyEvent (delegate Custom|Server%20Play%20Audio) (Math|Random|RandomFloatinRange _randtimerange_x _randtimerange_y)))

(event Custom|Client%20Play%20Audio
  (Networking|CanExecuteCosmeticEvents
    (:True
      (if (not (Variables|Default|GetbMarkForDelete))
        (Audio|PlaySoundatLocation (Variables|Default|GetSoundToPlay) (Transformation|GetActorLocation))))))

(fn Initialize
  (params (RandTimeRange :type Vector%202D%20Structure) (SoundToPlay :type Sound%20Base%20Object%20Reference))
  (bind (_randtimerange_x _randtimerange_y) (Math|Vector2D|BreakVector2D (Variables|GetRandTimeRange_X)))
  (Variables|Default|SetRandTimeRange RandTimeRange)
  (Variables|Default|SetSoundToPlay SoundToPlay)
  (if (Networking|IsDedicatedServer)
    (Utilities|Time|SetTimerbyEvent (EventDispatchers|CreateEvent :!event Server%20Play%20Audio) (Math|Random|RandomFloatinRange _randtimerange_x _randtimerange_y))))

(fn ConstructionScript)
