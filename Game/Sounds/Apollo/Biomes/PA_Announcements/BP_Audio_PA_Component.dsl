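; Actor component for Apollo biome PA announcement audio that manages a companion PA audio actor for its owner.
; On BeginPlay, only on a dedicated server, it spawns BP_Audio_PA_Actor at the component's world transform if RepActor is not already valid.
; It stores that actor in RepActor and initializes it with the random time range and sound to play.
; On EndPlay it destroys the spawned actor if still valid.
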(event EventBeginPlay
  (if (Networking|IsDedicatedServer)
    (Variables|Default|GetRepActor
      (:then)
      (:else
        (bind _returnvalue (Game|SpawnActorfromClass "/Game/Sounds/Apollo/Biomes/PA_Announcements/BP_Audio_PA_Actor.BP_Audio_PA_Actor_C" (Transformation|GetWorldTransform) "Undefined" "OverrideRootScale" (Components|GetOwner) 0))
        (Variables|Default|SetRepActor _returnvalue)
        (Class|BPAudioPAActor|Initialize _returnvalue (Variables|Default|GetRandTimeRange) (Variables|Default|GetSoundToPlay))))))

(event EventEndPlay
  (params (EndPlayReason :type EEndPlayReason%20Enum :default "Destroyed"))
  (bind _repactor (Variables|Default|GetRepActor)
    (:then
      (Actor|DestroyActor _repactor))
    (:else)))
